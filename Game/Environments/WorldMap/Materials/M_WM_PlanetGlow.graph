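# Additive, unlit, two-sided material for a planet glow effect on the world map, usable on particle sprites.
# A 2D sphere gradient serves as UVs into the GradientGlow texture, turning it into a radial color ramp.
# The sampled color is multiplied by particle alpha and fed into the shared MF_FX_EmissiveAdjustments function.
# An Opacity parameter exists but connects to nothing.
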Material /Game/Environments/WorldMap/Materials/M_WM_PlanetGlow
graph {
 properties {
  bTangentSpaceNormal=false
  bUsedWithParticleSprites=true
  blendMode=BLEND_Additive
  pixelDepthOffsetMode=PDOM_Legacy
  shadingModel=MSM_Unlit
  twoSided=true
  usedShadingModels=Unlit
 }
 nodes [
  n1 = MaterialFunctionCall!(UVs, CenterPosition, Radius=Constant() {r=0.5}) {materialFunction=/Engine/Functions/Engine_MaterialFunctions01/Gradient/SphereGradient-2D}
  MaterialFunctionCall(Emissive=Multiply(
    A=TextureSampleParameter2D(UVs=@n1.Result 0-1) {parameterName=GradientGlow, texture=/Game/Environments/WorldMap/Textures/WM_WorldGlow_D}.RGB
    B=ParticleColor().A
   )) {materialFunction=&MF_FX_EmissiveAdjustments}
  ScalarParameter() {defaultValue=0.4, parameterName=Opacity}
 ]
}
